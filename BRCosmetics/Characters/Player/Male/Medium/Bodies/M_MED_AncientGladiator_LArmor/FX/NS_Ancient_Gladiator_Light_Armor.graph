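# Niagara system for the Ancient Gladiator light armor cosmetic, adding purple glows to lion and eye details on the shoulder, belt and knees.
# Sprite emitters burst once, loop their particle lifetimes and follow bones or sockets on User.SkeletalMesh, using spawn groups to place each eye glow.
# Color curves pulse the eyes.
# The lion mouth emits drifting, fading mist alongside a light renderer driven by flicker curves.
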NiagaraSystem /BRCosmetics/Characters/Player/Male/Medium/Bodies/M_MED_AncientGladiator_LArmor/FX/NS_Ancient_Gladiator_Light_Armor
dependencies {
 usedDataInterfaces [&NiagaraDataInterfaceCurve, &NiagaraDataInterfaceColorCurve]
 usedDynamicInputs []
 usedModules [&SystemState]
}
emitters {
 EyeGlow_on_Belt {
  emitterSpawnScript [&FN_PreserveSkydiveUserVariables() {module=PreserveUserVariables}]
  emitterUpdateScript [
   &EmitterState()
   &SpawnBurst_Instantaneous(Spawn Count=3, Spawn Probability=HlslExpression() {hlslExpression="Emitter.LoopCount == 0"}, Use Spawn Probability=true)
  ]
  particleSpawnScript [
   &FN_SetSpawnGroup_v2(Number of Spawn Groups=3)
   /Niagara/Modules/Spawn/Initialization/InitializeParticle(
    Color=<rgba 1.7375 0 2 0.2>
    Lifetime=&RandomRangeFloat(Minimum=3, Maximum=4)
    Module.Write Color=true
    Module.Write Lifetime=true
    Module.Write Mass=false
    Module.Write Material Random=true
    Module.Write Position=false
    Module.Write Scale=true
    Module.Write SpriteRotation=false
    Module.Write SpriteSize=true
    Sprite Size=&Vector2DFromFloat(Value=HlslExpression() {hlslExpression="Particles.SpawnGroup == 0 ? 14 : 8"})
   )
   &CameraOffset(Camera Offset Amount=2)
   /SetVariables(Particles.LateralSpawnGroup=HlslExpression() {hlslExpression="Particles.SpawnGroup > 0 ? 1 : 0"}) {isSetParameters=true, module=SetVariables_#1}
   /Game/Effects/Niagara/Modules/Math/FN_EvenNumberToSign(Input:NiagaraInt32=$Particles.SpawnGroup, Output Float Attribute=true)
  ]
  particleUpdateScript [
   &ParticleState(Kill Particles When Lifetime Has Elapsed=false, Loop Particles Lifetime=true)
   &SkeletalMeshLocation(
    Apply Sampled Orientation to Mesh Particles=true
    Bone / Socket Index=4
    Bone Sampling Mode=::NewEnumerator4
    Inherit Velocity=false
    Interpolate Mesh Sampling=false
    MaskSpawnGroup=false
    Sampled Position Offset=&Add_Vector(
     A=<4.65 13 0>
     B=&MakeVector(
      X=&Multiply_Float(
       A=$Particles.LateralSpawnGroup
       B=&MaskFloatBySpawnGroup(Masked Float=0.5, Passthrough Float=-0.5, Spawn Group Mask=2)
      )
      Y=&Multiply_Float(A=$Particles.LateralSpawnGroup, B=-1.5)
      Z=&Multiply_Float(A=$Particles.Sign, B=&Multiply_Float(A=$Particles.LateralSpawnGroup, B=8))
     )
    )
    Sampled Position Offset Coordinate Space:ENiagaraMeshTransforms=::NewEnumerator3
    Skeletal Mesh=$User.SkeletalMesh
   )
   &ScaleColor(
    Scale Mode=::NewEnumerator1
    Scale RGBA=&MakeVector4_FromLinearColor(LinearColor=&ColorFromCurve(ColorCurve=DataInterface() {propertyValues {
        AlphaCurve {keys [<at 0 0.5>,<at 0.5 1>,<at 1 0.5>]}
        BlueCurve {keys [<at 0 1 RCIM_Cubic>,<at 0.5 2 RCIM_Cubic>,<at 1 1 RCIM_Cubic>]}
        GreenCurve {keys [<at 0 1 RCIM_Cubic>,<at 0.5 2 RCIM_Cubic>,<at 1 1 RCIM_Cubic>]}
        RedCurve {keys [<at 0 1 RCIM_Cubic>,<at 0.5 2 RCIM_Cubic>,<at 1 1 RCIM_Cubic>]}
       }}))
    ScaleRGBA=true
   )
  ]
  rendererClasses [&NiagaraSpriteRendererProperties]
  renderers [{rendererClass=&NiagaraSpriteRendererProperties}]
 }
 EyeGlow_on_Knees {
  emitterSpawnScript [&FN_PreserveSkydiveUserVariables() {module=PreserveUserVariables}]
  emitterUpdateScript [
   &EmitterState()
   &SpawnBurst_Instantaneous(Spawn Count=4, Spawn Probability=HlslExpression() {hlslExpression="Emitter.LoopCount == 0"}, Use Spawn Probability=true)
  ]
  particleSpawnScript [
   &InitializeParticle(
    Color=<rgba 1.7375 0 2 0.2>
    Lifetime=&RandomRangeFloat(Minimum=3, Maximum=4)
    Module.Write Color=true
    Module.Write Lifetime=true
    Module.Write Mass=false
    Module.Write Material Random=true
    Module.Write Position=false
    Module.Write Scale=true
    Module.Write SpriteRotation=false
    Module.Write SpriteSize=true
    Sprite Size=<8 8>
   )
   &FN_SetSpawnGroup_v2(Number of Spawn Groups=4)
   &CameraOffset(Camera Offset Amount=1)
   /SetVariables(
    Particles.LateralSpawnGroup=HlslExpression() {hlslExpression="Particles.SpawnGroup < 2 ? 1 : -1 "}
    Particles.LessThanSpawnGroup=HlslExpression() {hlslExpression="Particles.SpawnGroup < 2 ? 1 : 0"}
   ) {isSetParameters=true, module=SetVariables_#1}
   &FN_EvenNumberToSign(Input:NiagaraInt32=$Particles.SpawnGroup, Output Float Attribute=true)
  ]
  particleUpdateScript [
   &ParticleState(Kill Particles When Lifetime Has Elapsed=false, Loop Particles Lifetime=true)
   &SkeletalMeshLocation(
    Apply Sampled Orientation to Mesh Particles=true
    Bone / Socket Index=&MakeCustomIntFromBool(Bool=$Particles.LessThanSpawnGroup, True INT=2, False INT=3)
    Bone Sampling Mode=::NewEnumerator4
    Inherit Velocity=false
    Interpolate Mesh Sampling=false
    MaskSpawnGroup=false
    Sampled Position Offset=&MakeVector(
     X=&Multiply_Float(A=$Particles.LateralSpawnGroup, B=13)
     Y=&Multiply_Float(A=$Particles.LateralSpawnGroup, B=10)
     Z=&Multiply_Float(A=$Particles.Sign, B=2)
    )
    Sampled Position Offset Coordinate Space:ENiagaraMeshTransforms=::NewEnumerator3
    Skeletal Mesh=$User.SkeletalMesh
   )
   &ScaleColor(
    Scale Mode=::NewEnumerator1
    Scale RGBA=&MakeVector4_FromLinearColor(LinearColor=&ColorFromCurve(ColorCurve=DataInterface() {propertyValues {
        AlphaCurve {keys [<at 0 0.5>,<at 0.5 1>,<at 1 0.5>]}
        BlueCurve {keys [<at 0 1 RCIM_Cubic>,<at 0.5 2 RCIM_Cubic>,<at 1 1 RCIM_Cubic>]}
        GreenCurve {keys [<at 0 1 RCIM_Cubic>,<at 0.5 2 RCIM_Cubic>,<at 1 1 RCIM_Cubic>]}
        RedCurve {keys [<at 0 1 RCIM_Cubic>,<at 0.5 2 RCIM_Cubic>,<at 1 1 RCIM_Cubic>]}
       }}))
    ScaleRGBA=true
   )
  ]
  rendererClasses [&NiagaraSpriteRendererProperties]
  renderers [{rendererClass=&NiagaraSpriteRendererProperties}]
 }
 EyeGlow_on_LionShoulder {
  emitterSpawnScript [&FN_PreserveSkydiveUserVariables() {module=PreserveUserVariables}]
  emitterUpdateScript [
   &EmitterState()
   &SpawnBurst_Instantaneous(Spawn Count=2, Spawn Probability=HlslExpression() {hlslExpression="Emitter.LoopCount == 0"}, Use Spawn Probability=true)
  ]
  particleSpawnScript [
   &FN_SetSpawnGroup_v2(Number of Spawn Groups=2)
   &InitializeParticle(
    Color=<rgba 1.7375 0 2 0.35>
    Lifetime=&RandomRangeFloat(Minimum=3, Maximum=4)
    Module.Write Color=true
    Module.Write Lifetime=true
    Module.Write Mass=false
    Module.Write Material Random=true
    Module.Write Position=false
    Module.Write Scale=false
    Module.Write SpriteRotation=false
    Module.Write SpriteSize=true
    Sprite Size=&Vector2DFromFloat(Value=12)
   )
   &CameraOffset(Camera Offset Amount=2)
  ]
  particleUpdateScript [
   &ParticleState(Kill Particles When Lifetime Has Elapsed=false, Loop Particles Lifetime=true)
   &SkeletalMeshLocation(
    Apply Sampled Orientation to Mesh Particles=true
    Bone / Socket Index=1
    Bone Sampling Mode=::NewEnumerator4
    Inherit Velocity=false
    Interpolate Mesh Sampling=false
    MaskSpawnGroup=false
    Sampled Position Offset=&MaskVectorBySpawnGroup(Masked Vector=<1.5 2.5 11>, Passthrough Vector=<1.5 -5.25 11>)
    Sampled Position Offset Coordinate Space:ENiagaraMeshTransforms=::NewEnumerator3
    Skeletal Mesh=$User.SkeletalMesh
   )
   &ScaleColor(
    Scale Mode=::NewEnumerator1
    Scale RGBA=&MakeVector4_FromLinearColor(LinearColor=&ColorFromCurve(ColorCurve=DataInterface() {propertyValues {
        AlphaCurve {keys [<at 0 0.5>,<at 0.5 1>,<at 1 0.5>]}
        BlueCurve {keys [<at 0 1 RCIM_Cubic>,<at 0.5 2 RCIM_Cubic>,<at 1 1 RCIM_Cubic>]}
        GreenCurve {keys [<at 0 1 RCIM_Cubic>,<at 0.5 2 RCIM_Cubic>,<at 1 1 RCIM_Cubic>]}
        RedCurve {keys [<at 0 1 RCIM_Cubic>,<at 0.5 2 RCIM_Cubic>,<at 1 1 RCIM_Cubic>]}
       }}))
    ScaleRGBA=true
   )
  ]
  rendererClasses [&NiagaraSpriteRendererProperties]
  renderers [{rendererClass=&NiagaraSpriteRendererProperties}]
 }
 LionMouthGlowAndMist {
  emitterUpdateScript [
   &EmitterState()
   &SpawnRate(SpawnRate=20, Use Spawn Probability=false)
  ]
  particleSpawnScript [
   &InitializeParticle(
    Color=&MakeLinearColorFromVectorAndFloat(
     "Vector (RGB)"=&MakeVectorFromLinearColorRGB(Color=<rgba 1.0313 0 1.25 1>)
     "Float (Alpha)"=&RandomRangeFloat(Minimum=0.25, Maximum=0.5)
    )
    Lifetime=1.5
    Mesh Scale=&VectorFromFloat(Value=1)
    Module.Write Color=true
    Module.Write Lifetime=true
    Module.Write Mass=true
    Module.Write Material Random=true
    Module.Write Position=true
    Module.Write Scale=true
    Module.Write SpriteRotation=true
    Module.Write SpriteSize=true
    Sprite Rotation=&UniformRangedFloat(Minimum=0, Maximum=360)
    Sprite Size=&Vector2DFromFloat(Value=&RandomRangeFloat(Minimum=8, Maximum=12))
   )
   &CameraOffset(Camera Offset Amount=&RandomRangeFloat(Minimum=0, Maximum=1))
  ]
  particleUpdateScript [
   &ParticleState()
   &DynamicMaterialParameters(
    Index 0 Param 1=&Multiply_Float(A=0.2, B=$Engine.Time)
    Index 0 Param 2=&RandomRangeFloat(Minimum=0, Maximum=1)
    Index 0 Param 3=1
    Index 0 Param 4=&RandomRangeFloat(Minimum=0.25, Maximum=8)
    Module.Param0WriteEnabled=true
    Module.Param1WriteEnabled=true
    Module.Param2WriteEnabled=true
    Module.Param3WriteEnabled=true
   )
   &SkeletalMeshLocation(
    Apply Sampled Orientation to Mesh Particles=true
    Bone / Socket Index=1
    Bone Sampling Mode=::NewEnumerator4
    Inherit Velocity=false
    MaskSpawnGroup=false
    Sampled Position Offset=&Add_Vector(A=<2 -2 6>, B=&Multiply_VectorByFloat(Vector=<8 0 3>, Float=$Particles.NormalizedAge))
    Sampled Position Offset Coordinate Space:ENiagaraMeshTransforms=::NewEnumerator3
    Skeletal Mesh=$User.SkeletalMesh
   )
   &ScaleColor(
    Scale Alpha=&FloatFromCurve(FloatCurve=DataInterface() {propertyValues {Curve {keys [
         {arriveTangent=4.7744, interpMode=RCIM_Cubic, leaveTangent=4.7744, tangentMode=RCTM_User, time=0, value=0}
         {arriveTangent=-0.2362, interpMode=RCIM_Cubic, leaveTangent=-0.2362, tangentMode=RCTM_User, time=0.5827, value=0.9932}
         {arriveTangent=-3.3493, interpMode=RCIM_Cubic, leaveTangent=-3.3493, tangentMode=RCTM_User, time=1, value=0}
        ]}}})
    Scale RGB=&Lerp_Vector(A=<2 2 2>, B=&MakeVectorFromLinearColorRGB(Color=<rgba 0.4102 0 1 1>), Alpha=$Particles.NormalizedAge)
    ScaleA=true
    ScaleRGB=true
   )
  ]
  rendererClasses [&NiagaraSpriteRendererProperties]
  renderers [{rendererClass=&NiagaraSpriteRendererProperties}]
 }
 LionMouth_Light {
  emitterUpdateScript [
   &EmitterState()
   &SpawnBurst_Instantaneous(Spawn Count=1, Spawn Probability=HlslExpression() {hlslExpression="Emitter.LoopCount == 0"}, Use Spawn Probability=true)
  ]
  particleSpawnScript [&InitializeParticle(
    Color=&MakeLinearColorFromVectorAndFloat(
     "Vector (RGB)"=&MakeVectorFromLinearColorRGB(Color=<rgba 1.0313 0 1.25 1>)
     "Float (Alpha)"=&RandomRangeFloat(Minimum=0.25, Maximum=0.5)
    )
    Lifetime=1.5
    Mesh Scale=&VectorFromFloat(Value=1)
    Module.Write Color=true
    Module.Write Lifetime=true
    Module.Write Mass=true
    Module.Write Material Random=true
    Module.Write Position=true
    Module.Write Scale=true
    Module.Write SpriteRotation=true
    Module.Write SpriteSize=true
    Sprite Rotation=&UniformRangedFloat(Minimum=0, Maximum=360)
    Sprite Size=&Vector2DFromFloat(Value=&RandomRangeFloat(Minimum=8, Maximum=12))
   )]
  particleUpdateScript [
   &ParticleState(
    Kill Particles When Lifetime Has Elapsed=false
    Let Infinitely Lived Particles Die When Emitter Deactivates=true
    Loop Particles Lifetime=true
   )
   &SkeletalMeshLocation(
    Apply Sampled Orientation to Mesh Particles=true
    Bone / Socket Index=1
    Bone Sampling Mode=::NewEnumerator4
    Inherit Velocity=false
    MaskSpawnGroup=false
    Sampled Position Offset=<4 -2.5 6>
    Sampled Position Offset Coordinate Space:ENiagaraMeshTransforms=::NewEnumerator3
    Skeletal Mesh=$User.SkeletalMesh
   )
   &FN_Light_v3(
    Color=<rgba 1.2844 0 10 1>
    Create Color Attribute=true
    Create Exponent Attribute=true
    Create Position Attribute=true
    Create Radius Attribute=true
    Exponent=1
    Position:Vector3f=&Add_Vector(
     A=$Output.SkeletalMeshLocation.SampledPosition
     B=&RotateVectorByQuaternion(Vector To Rotate=<6 -2 8>, Quaternion=$Output.SkeletalMeshLocation.SampledRotation)
    )
    Radius=6
    Scale Color:NiagaraFloat=&FloatFromCurve(FloatCurve=DataInterface() {propertyValues {Curve {keys [<at 0 1>,<at 0.5 2>,<at 1 1>]}}})
   )
   /SetVariables(
    Particles.LightComponentColor:LinearColor=<rgba 0.4 0 2 1>
    Particles.LightComponentIntensity:NiagaraFloat=&FloatFromCurve(
     FloatCurve=DataInterface() {propertyValues {Curve {keys [
         <at 0 1>
         {arriveTangent=-1.0745, interpMode=RCIM_Cubic, leaveTangent=-1.0745, time=0.0728, value=1}
         {arriveTangent=0.0825, interpMode=RCIM_Cubic, leaveTangent=0.0825, time=0.1439, value=0.8454}
         <at 0.2119 1.0115 RCIM_Cubic>
         {arriveTangent=0.7006, interpMode=RCIM_Cubic, leaveTangent=0.7006, time=0.277, value=0.8454}
         {arriveTangent=2.982, interpMode=RCIM_Cubic, leaveTangent=2.982, time=0.3375, value=1.0994}
         {arriveTangent=-0.401, interpMode=RCIM_Cubic, leaveTangent=-0.401, time=0.4539, value=1.373}
         {arriveTangent=-0.9416, interpMode=RCIM_Cubic, leaveTangent=-0.9416, time=0.5567, value=1.0115}
         {arriveTangent=-0.5635, interpMode=RCIM_Cubic, leaveTangent=-0.5635, time=0.7548, value=1.0896}
         {arriveTangent=-0.2904, interpMode=RCIM_Cubic, leaveTangent=-0.2904, time=0.8168, value=0.8649}
         {arriveTangent=0.7374, interpMode=RCIM_Cubic, leaveTangent=0.7374, time=0.8894, value=1.0506}
         <at 1 1>
        ]}}}
     CurveIndex=$Emitter.NormalizedLoopAge
     Scale Curve=4
    )
   ) {isSetParameters=true}
  ]
  rendererClasses [&NiagaraLightRendererProperties]
  renderers [{rendererClass=&NiagaraLightRendererProperties, rendererIndex=0}]
 }
}
summary {
 systemName=NS_Ancient_Gladiator_Light_Armor
 emitters [
  {emitterName=LionMouthGlowAndMist, rendererClasses [&NiagaraSpriteRendererProperties]}
  {emitterName=EyeGlow_on_Knees, rendererClasses [&NiagaraSpriteRendererProperties]}
  {emitterName=EyeGlow_on_Belt, rendererClasses [&NiagaraSpriteRendererProperties]}
  {emitterName=EyeGlow_on_LionShoulder, rendererClasses [&NiagaraSpriteRendererProperties]}
  {emitterName=LionMouth_Light, rendererClasses [&NiagaraLightRendererProperties]}
 ]
 userVariables [
  User.bIsFrontEndPreview {type=NiagaraBool}
  User.Interpolate_Rate_of_Change {type=NiagaraFloat, defaultValue=7, description=""}
  User.Interpolate_WPO {type=NiagaraFloat, defaultValue=1, description=""}
  User.PlayerSpeed {type=NiagaraFloat, defaultValue=0, description=""}
  User.SkeletalMesh {
   type=NiagaraDataInterfaceSkeletalMesh
   defaultValue {
    struct=NiagaraExt_VariableValue_DataInterface
    value {dataInterface=&self:NiagaraDataInterfaceSkeletalMesh_1}
   }
  }
 ]
}
