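# Niagara system for a purple fire plume on the Ancient Gladiator light-armor helmet cosmetic, anchored to the character's skeletal mesh.
# Fire_Plume_World spawns short-lived animated sprites that follow the sampled bone while blending slightly toward their spawn position with age.
# Flame_Plume_and_Light keeps one bone-oriented mesh, writing a smoothed position and WPO amount to a dynamic attribute for motion lag.
# User.bisHeadHidden scales both emitters to zero.
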NiagaraSystem /BRCosmetics/Characters/Player/Male/Medium/Bodies/M_MED_AncientGladiator_LArmor/FX/NS_Ancient_Gladiator_Light_Helmet
dependencies {
 usedDataInterfaces [&NiagaraDataInterfaceCurve, &NiagaraDataInterfaceVectorCurve, &NiagaraDataInterfaceColorCurve]
 usedDynamicInputs [&VectorFromCurve]
 usedModules [&SystemState]
}
emitters {
 Fire_Plume_World {
  emitterUpdateScript [
   &EmitterState()
   &SpawnRate(SpawnRate=12, Use Spawn Probability=false)
  ]
  particleSpawnScript [
   /Niagara/Modules/Spawn/Initialization/InitializeParticle(
    Color=&RandomRangeLinearColor(Minimum=<rgba 0.325 0 0.65 0.5>, Maximum=<rgba 1.2 0 2 1>)
    Lifetime=&RandomRangeFloat(Minimum=1, Maximum=1.25)
    Module.Write Color=true
    Module.Write Lifetime=true
    Module.Write Mass=false
    Module.Write Material Random=true
    Module.Write Position=false
    Module.Write Scale=false
    Module.Write SpriteRotation=true
    Module.Write SpriteSize=true
    Position=&SimulationPosition()
    Sprite Rotation=&NormalizedAngleToDegrees(NormalizedAngle=&RandomRangeFloat(Minimum=-1, Maximum=1))
    Sprite Size=&Vector2DFromFloat(Value=&RandomRangeFloat(Minimum=24, Maximum=26))
   )
   &SkeletalMeshLocation(
    Bone Sampling Mode=::NewEnumerator4
    Inherit Velocity=false
    Interpolate Mesh Sampling=false
    MaskSpawnGroup=false
    Sampled Position Offset Coordinate Space:ENiagaraMeshTransforms=::NewEnumerator3
    Skeletal Mesh=$User.SkeletalMesh
   )
   /SetVariables(Particles.SpawnVector:Vector3f=&MakeVector(
     X=&FloatFromCurve(
      FloatCurve=DataInterface() {propertyValues {Curve {keys [
          {arriveTangent=8.098, interpMode=RCIM_Cubic, time=0, value=14}
          {arriveTangent=-43.5023, interpMode=RCIM_Cubic, leaveTangent=-43.5025, tangentMode=RCTM_User, time=1, value=-2}
         ]}}}
      CurveIndex=$Particles.MaterialRandom
     )
     Y=&FloatFromCurve(FloatCurve=DataInterface() {propertyValues {Curve {keys [<at 0 10>,<at 1 -16>]}}}, CurveIndex=$Particles.MaterialRandom)
     Z=0
    )) {isSetParameters=true}
   /SetVariables(Particles.RenderPosition:Vector3f=&RotateVectorByQuaternion(
     Vector To Rotate=&Add_Vector(A=$Particles.SpawnVector, B=<10 0 0>)
     Quaternion=$Output.SkeletalMeshLocation.SampledRotation
    )) {isSetParameters=true}
   &CameraOffset(Camera Offset Amount=-3)
  ]
  particleUpdateScript [
   &ParticleState()
   &SkeletalMeshLocation(
    Apply To Particle Position=false
    Bone Sampling Mode=::NewEnumerator4
    Inherit Velocity=false
    Interpolate Mesh Sampling=false
    MaskSpawnGroup=false
    Sampled Position Offset Coordinate Space:ENiagaraMeshTransforms=::NewEnumerator3
    Skeletal Mesh=$User.SkeletalMesh
   ) {module=SkeletalMeshLocation001}
   &ScaleColor(
    Scale Mode=::NewEnumerator1
    Scale RGBA=&Vector4FromCurve(Vector4Curve=DataInterface() {propertyValues {
       AlphaCurve {keys [
         <at 0 0 RCIM_Cubic>
         {arriveTangent=3.3333, interpMode=RCIM_Cubic, leaveTangent=3.3333, time=0.3, value=1}
         {arriveTangent=-1.6667, interpMode=RCIM_Cubic, leaveTangent=-1.6667, time=0.6, value=2}
         {arriveTangent=-0.4881, interpMode=RCIM_Cubic, leaveTangent=-0.4881, tangentMode=RCTM_User, time=0.9, value=0}
        ]}
       BlueCurve {keys [<at 0.206 1>,<at 0.9 2>,<at 1.2198 0.4062>]}
       GreenCurve {keys [<at 0.206 1>,<at 0.9 2>,<at 1.2198 0.4062>]}
       RedCurve {keys [<at 0.206 1>,<at 0.9 2>,<at 1.2198 0.4062>]}
      }})
    ScaleRGBA=true
   )
   &DynamicMaterialParameters(
    Index 0 Param 1=&FloatFromCurve(FloatCurve=DataInterface() {propertyValues {Curve {keys [{arriveTangent=2.7551, interpMode=RCIM_Cubic, leaveTangent=2.7551, tangentMode=RCTM_User, time=0, value=0}, <at 0.6 1>]}}})
    Index 0 Param 2=&RandomRangeFloat(Minimum=0.33, Maximum=1)
    Index 0 Param 3=&FloatFromCurve(FloatCurve=DataInterface() {propertyValues {Curve {keys [
         {arriveTangent=6.3026, interpMode=RCIM_Cubic, time=0, value=1}
         {arriveTangent=2.4291, interpMode=RCIM_Cubic, leaveTangent=6.3584, tangentMode=RCTM_User, time=1, value=2}
        ]}}})
    Index 0 Param 4=&RandomRangeFloat(Minimum=1, Maximum=2)
    Module.Param0WriteEnabled=true
    Module.Param1WriteEnabled=true
    Module.Param2WriteEnabled=true
    Module.Param3WriteEnabled=true
   ) {module=DynamicMaterialParameters001}
   &ScaleSpriteSize(Scale Factor=&Vector2DFromFloat(Value=&FloatFromCurve(
      FloatCurve=DataInterface() {propertyValues {Curve {keys [
          {arriveTangent=-0.0804, interpMode=RCIM_Cubic, leaveTangent=-0.0804, tangentMode=RCTM_User, time=0, value=0.75}
          {interpMode=RCIM_Cubic, tangentMode=RCTM_User, time=0.4, value=1}
          {arriveTangent=-4.5991, interpMode=RCIM_Cubic, leaveTangent=-12.9762, tangentMode=RCTM_User, time=1, value=0}
         ]}}}
      Scale Curve=&MakeCustomFloatFromBool(Bool=$User.bisHeadHidden, True Float=0, False Float=1)
     )))
   &SpriteRotationRate(Rotation Rate=20)
   &self:Location_LERP_-_Source_to_Destination_02(
    DestinationPosition:Vector3f=&Add_Vector(A=Linked() {linkedVariable {name=Particles.Initial.Position, type=NiagaraPosition}}, B=$Particles.RenderPosition)
    DestinationPosition Offset:Vector3f=&RotateVectorByQuaternion(
     Vector To Rotate=&Multiply_VectorByFloat(Vector=&RandomRangeVector(Minimum=<16 0 0>, Maximum=<22 0 0>), Float=$Particles.NormalizedAge)
     Quaternion=$Output.SkeletalMeshLocation001.SampledRotation
    )
    Normalized Influence:NiagaraFloat=&Lerp_Float(A=0, B=&RandomRangeFloat(Minimum=0, Maximum=0.125), Alpha=$Particles.NormalizedAge)
    SourcePosition:Vector3f=&Add_Vector(A=$Output.SkeletalMeshLocation001.SampledPosition, B=$Particles.RenderPosition)
    SourcePositionOffset:Vector3f=&RotateVectorByQuaternion(Vector To Rotate=<3 0 0>, Quaternion=$Output.SkeletalMeshLocation001.SampledRotation)
   ) {module=Location_LERP_-_Source_to_Destination_02}
   &SubUVAnimation()
  ]
  rendererClasses [&NiagaraSpriteRendererProperties]
  renderers [{rendererClass=&NiagaraSpriteRendererProperties}]
 }
 Flame_Plume_and_Light {
  emitterUpdateScript [
   &EmitterState()
   &SpawnBurst_Instantaneous(Spawn Count=1, Spawn Probability=HlslExpression() {hlslExpression="Emitter.LoopCount == 0"}, Use Spawn Probability=true)
  ]
  particleSpawnScript [
   &InitializeParticle(
    Color=<rgba 0.8 0 1.5 1>
    Lifetime=4
    Module.Write Color=true
    Module.Write Lifetime=true
    Module.Write Mass=true
    Module.Write Material Random=true
    Module.Write Position=false
    Module.Write Scale=true
    Module.Write SpriteRotation=false
    Module.Write SpriteSize=false
    Position=&SimulationPosition()
   )
   &DynamicMaterialParameters(
    Index 0 Param 3=22
    Module.Param0WriteEnabled=true
    Module.Param1WriteEnabled=true
    Module.Param2WriteEnabled=true
    Module.Param3WriteEnabled=true
   )
  ]
  particleUpdateScript [
   &ParticleState(Kill Particles When Lifetime Has Elapsed=false, Let Infinitely Lived Particles Die When Emitter Deactivates=true)
   &ScaleMeshSize(Scale Factor=&VectorFromFloat(Value=&MakeCustomFloatFromBool(Bool=$User.bisHeadHidden, True Float=0, False Float=1)))
   &InterpolateOverTime(
    Rate Of Change:NiagaraFloat=$User.Interpolate_Rate_of_Change
    Target Vector Value:NiagaraPosition=$Particles.Position
    Variable Type=::NewEnumerator2
   )
   /SetVariables(Particles.Interpolate Dynamic=&MakeFloat4FromVectorAndFloat("Vector (RGB)"=$Output.InterpolateOverTime.Moving Average, "Float (Alpha)"=$User.Interpolate_WPO)) {isSetParameters=true}
   &SkeletalMeshLocation(
    Apply Sampled Orientation to Mesh Particles=true
    Bone Sampling Mode=::NewEnumerator4
    Inherit Velocity=false
    Interpolate Mesh Sampling=false
    MaskSpawnGroup=false
    Sampled Position Offset=<21 -1 0>
    Sampled Position Offset Coordinate Space:ENiagaraMeshTransforms=::NewEnumerator3
    Skeletal Mesh=$User.SkeletalMesh
   )
   /SetVariables(Particles.MeshOrientation=&MultiplyQuaternion(
     Quaternion A=$Output.SkeletalMeshLocation.SampledRotation
     Quaternion B=&FindQuatBetweenTwoVectors(From Vector=<0 0 1>, To Vector=<1 0 0>)
    )) {isSetParameters=true}
   &ScaleColor(
    Scale Alpha=&MakeCustomFloatFromBool(Bool=$User.bisHeadHidden, True Float=0, False Float=1)
    ScaleA=true
    ScaleRGB=true
   )
  ]
  rendererClasses [&NiagaraMeshRendererProperties]
  renderers [{rendererClass=&NiagaraMeshRendererProperties, rendererIndex=0}]
 }
}
summary {
 systemName=NS_Ancient_Gladiator_Light_Helmet
 emitters [
  {emitterName=Fire_Plume_World, rendererClasses [&NiagaraSpriteRendererProperties]}
  {emitterName=Flame_Plume_and_Light, rendererClasses [&NiagaraMeshRendererProperties]}
 ]
 userVariables [
  User.bIsFrontEndPreview {type=NiagaraBool}
  User.bisHeadHidden {type=NiagaraBool}
  User.Interpolate_Rate_of_Change {type=NiagaraFloat, defaultValue=7, description=""}
  User.Interpolate_WPO {type=NiagaraFloat, defaultValue=1, description=""}
  User.PlayerSpeed {type=NiagaraFloat, defaultValue=0, description=""}
  User.SkeletalMesh {
   type=NiagaraDataInterfaceSkeletalMesh
   defaultValue {
    struct=NiagaraExt_VariableValue_DataInterface
    value {dataInterface=&self:NiagaraDataInterfaceSkeletalMesh_1}
   }
  }
 ]
}
